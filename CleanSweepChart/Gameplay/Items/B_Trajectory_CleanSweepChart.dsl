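; Trajectory Blueprint for the CleanSweepChart item, extending a parent trajectory class.
; Its UpdateFromTrajectoryOwner override runs the parent logic, then sets the Target component's world rotation from the instigator's yaw offset by a quarter turn, with pitch and roll zeroed.
; This keeps the target flat and aligned relative to the instigator's heading.
; The construction script only calls the parent, and the event graph holds only disabled stubs.
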(event EventBeginPlay
  (|CallParentFunction) ; This node is disabled and will not be called.\nDrag off pins to build functionality.
)

(event Collision|EventActorBeginOverlap
  (params (OtherActor :type Actor%20Object%20Reference))
  (|CallParentFunction OtherActor) ; This node is disabled and will not be called.\nDrag off pins to build functionality.
)

(event EventTick
  (params (DeltaSeconds :type Float%20%28single-precision%29))
  (|CallParentFunction DeltaSeconds) ; This node is disabled and will not be called.\nDrag off pins to build functionality.
)

(fn UpdateFromTrajectoryOwner
  (bind (_returnvalue_roll _returnvalue_pitch _returnvalue_yaw) (Math|Rotator|BreakRotator (Transformation|GetActorRotation (Variables|Actor|GetInstigator))))
  (|CallParentFunction)
  (Transformation|SetWorldRotation (Variables|Default|GetTarget) (Math|Rotator|MakeRotator 0.0 0.0 (+ _returnvalue_yaw -9e+01)) false false))

(fn ConstructionScript
  (|CallParentFunction))
